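# Physics asset for the male medium Genius Blob cosmetic head, defining simulation bodies and joints for three dynamic hair bones.
# Each hair bone has an offset capsule and is constrained to its own MSTR bone, which carries a tiny capsule as an anchor point.
# The joints allow wide swing, tuned per strand, with very tight twist, so each piece bends around its base with little roll.
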PhysicsAsset /BRCosmetics/Characters/Player/Male/Medium/Heads/M_MED_Genius_Blob_Head/Meshes/M_MED_Genius_Blob_Head_Physics
bodies {
 dyn_hair_01 [{
   shapeType=Capsule
   length=3.9072
   radius=1.5757
   shapeName=dyn_top_hair_02_capsule
   center=<4.6767 -0.0584 0>
   rotation=<rot 90 180 -179.3401>
  }]
 dyn_hair_01_MSTR [{shapeType=Capsule, length=0.505, radius=0.505, shapeName=dyn_hair_01_MSTR_capsule, rotation=<rot 0 -90 0>}]
 dyn_hair_02 [{
   shapeType=Capsule
   length=4.4486
   radius=1.8395
   shapeName=dyn_top_hair_02_capsule
   center=<5.1967 -0.1327 -0.0832>
   rotation=<rot 88.4021 179.9987 179.9987>
  }]
 dyn_hair_02_MSTR [{shapeType=Capsule, length=0.505, radius=0.505, shapeName=dyn_hair_02_MSTR_capsule, rotation=<rot 0 -90 0>}]
 dyn_hair_03 [{
   shapeType=Capsule
   length=2.2352
   radius=1.7454
   shapeName=dyn_top_hair_02_capsule
   center=<3.6823 -0.2013 0>
   rotation=<rot 90 180 -176.4444>
  }]
 dyn_hair_03_MSTR [{shapeType=Capsule, length=0.505, radius=0.505, shapeName=dyn_hair_03_MSTR_capsule, rotation=<rot 0 -90 0>}]
}
constraints [
 {bone1Name=dyn_hair_01, bone2Name=dyn_hair_01_MSTR, swing1LimitDegrees=70, swing2LimitDegrees=70, twistLimitDegrees=2}
 {bone1Name=dyn_hair_02, bone2Name=dyn_hair_02_MSTR, swing1LimitDegrees=75, swing2LimitDegrees=75, twistLimitDegrees=2}
 {bone1Name=dyn_hair_03, bone2Name=dyn_hair_03_MSTR, swing1LimitDegrees=60, swing2LimitDegrees=60, twistLimitDegrees=2}
]
